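# HUD widget for the Creative Capture Area device, showing capture status text, zone icon, allied player count and progress bar.
# On construct it binds to the owning player's CRD_CaptureProgressOnHUDComp for zone, progress and count updates.
# A dynamic material drives bar fill, team colors from the CoreGamer data asset, and state-based pulse and direction.
# Entering Captured starts a timer that plays the hide animation; state changes cancel pending hides.
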WidgetBlueprint /Game/Creative/Devices/CaptureArea/UI/CaptureArea_CaptureProgressOnHUD_Widget
animations [HideWidgetAnimation {
  binding_count=1
  length_seconds=0.2468
  object=&self:HideWidgetAnimation
  sequence {bindings [WidgetSizeBox {
     id=#1
     possessed_class=SizeBox
     tracks [
      {
       type=MovieSceneFloatTrack
       display_name=Render Opacity
       sections [{type=MovieSceneFloatSection, channels [None {type=MovieSceneScriptingFloatChannel, keys [[0,1],[4.9363,0]]}]}]
      }
      {
       type=MovieScene2DTransformTrack
       sections [{
         type=MovieScene2DTransformSection
         channels [
          Translation.X {type=MovieSceneScriptingFloatChannel, keys [[0,0],[4.9363,160]]}
          Translation.Y {type=MovieSceneScriptingFloatChannel, keys [[0,0],[4.9363,-5]]}
          Scale.X {type=MovieSceneScriptingFloatChannel, keys [[0,1],[4.9363,0.85]]}
          Scale.Y {type=MovieSceneScriptingFloatChannel, keys [[0,1],[4.9363,0.85]]}
         ]
        }]
      }
     ]
    }]}
 }]
graphs [
 EventGraph {nodes [
   n1 = Update Progress!() -> (OutputDelegate, then, Value:double=~, CaptureState:ECaptureState Enum=Neutral, CaptureStateChanged:bool=~) {is=title K2Node_CustomEvent}
   > Set MostRecentCaptureState(MostRecentCaptureState:ECaptureState Enum=@n1.CaptureState) -> (Output_Get:ECaptureState Enum=Neutral)
   > n2 = Branch(Condition=@n1.CaptureStateChanged)
   > n3 = ShowOrHideOnState()
   UpdateCaptureBarProgress!(execute=[@n3.then, @n2.else], self, Progress:double=@n1.Value) -> (then)
   > UpdateCaptureBarText()
   > SetVisibility(self=Get RootOverlay() -> (RootOverlay:Overlay Object Reference), InVisibility=HitTestInvisible)
   Remove Widget() {is=title K2Node_CustomEvent}
   > SetVisibility(InVisibility=Collapsed)
   > PopContentWidget!(execute) -> (then)
   n4 = UpdatePlayerCount!() -> (OutputDelegate, then, PlayerCount:int) {is=title K2Node_CustomEvent}
   > SetText(self=Get PlayerCount() -> (PlayerCount:Common Rich Text Block Object Reference), InText="To Text (Integer)"(Value=@n4.PlayerCount))
   ShowOrHideOnState() {is=title K2Node_CustomEvent}
   > n5 = Branch(Condition="Equal (Enum)"(A:ECaptureState Enum=Get MostRecentCaptureState() -> (MostRecentCaptureState:ECaptureState Enum=Neutral), B:ECaptureState Enum=Captured))
   > n6 = Sequence()
   StopAnimation(execute=[@n6.then_0, @n5.else], InAnimation=Get HideWidgetAnimation() -> (HideWidgetAnimation:Widget Animation Object Reference))
   > Clear and Invalidate Timer by Handle(Handle=Get PlayHideAnimTimerHandle() -> (PlayHideAnimTimerHandle:Timer Handle Structure))
   > SetRenderScale!(execute, self:Widget Object Reference=Get WidgetSizeBox() -> (WidgetSizeBox:Size Box Object Reference), Scale:Vector 2D Structure=(X=1,Y=1)) -> (then)
   n7 = PlayHideAnimation() {is=title K2Node_CustomEvent}
   n8 = Set Timer by Event(execute=@n6.then_1, Delegate=@n7, Time=Get DelayBeforeHidingOnCapture() -> (DelayBeforeHidingOnCapture:float), bLooping=false)
   > Set PlayHideAnimTimerHandle(PlayHideAnimTimerHandle:Timer Handle Structure=@n8) -> (Output_Get:Timer Handle Structure)
   Event PreConstruct()
   > n9 = GetDynamicMaterial(self=Get Image_Progress() -> (Image_Progress:Image Object Reference))
   > Set CaptureBarMaterial(CaptureBarMaterial:Material Instance Dynamic Object Reference=@n9) -> (Output_Get:Material Instance Dynamic Object Reference)
   > UpdatePulseState!(execute, self, bIsPulseActive:bool=true) -> (then)
   Event Construct()
   n10 = GetComponentByClass!(
    self:Actor Object Reference=GetOwningPlayer()
    ComponentClass:Actor Component Class Reference=/Game/Creative/Devices/CaptureArea/CRD_CaptureProgressOnHUDComp.CRD_CaptureProgressOnHUDComp_C
   ) -> (ReturnValue:CRD Capture Progress On HUDComp Object Reference)
   > Is Valid(InputObject=@n10)
   > Set CaptureProgressOnHUDComponent(CaptureProgressOnHUDComponent:CRD Capture Progress On HUDComp Object Reference=@n10) -> (Output_Get:CRD Capture Progress On HUDComp Object Reference)
   > n11 = Sequence()
   n12 = Get CaptureProgressOnHUDComponent() -> (CaptureProgressOnHUDComponent:CRD Capture Progress On HUDComp Object Reference)
   Bind Event to On Capture Zone Changed(execute=@n11.then_1, self:CRD Capture Progress On HUDComp Object Reference=@n12, Delegate=Create Event()) {is=title K2Node_AddDelegate}
   > Bind Event to On Update Progress(self:CRD Capture Progress On HUDComp Object Reference=@n12, Delegate=Create Event()) {is=title K2Node_AddDelegate}
   > Bind Event to On Update Player Count(self:CRD Capture Progress On HUDComp Object Reference=@n12, Delegate=Create Event()) {is=title K2Node_AddDelegate}
   > OnForceCaptureStateChange(self:CRD Capture Progress On HUDComp Object Reference=@n12)
   n13 = Get CaptureProgressOnHUDComponent()
   UpdatePlayerCount!(
    execute=@n11.then_0
    self
    PlayerCount:int=Get AlliesInCapZone!(self:CRD Capture Progress On HUDComp Object Reference=@n13) -> (AlliesInCapZone:int=0)
   ) -> (then) {is=title K2Node_CallFunction}
   > n14 = GetDataFromBlueprint!(
    execute:Exec
    self:Object Reference=Get ActiveCapAreaReference!(self:CRD Capture Progress On HUDComp Object Reference=@n13) -> (ActiveCapAreaReference:Building Prop Capture Area Object Reference)
   ) -> (then:Exec, IconOption:Creative Icon Option Structure) {is=title K2Node_Message}
   > OnCaptureZoneChanged!(
    execute
    self
    CapturePointIcon:Texture 2D Soft Object Reference=Get Icon Library Texture!(Icon Option:"Creative Icon Option Structure (by ref)"=@n14, Should Use Large Icon:bool=false) -> (Icon Texture:Texture 2D Soft Object Reference) {is=title K2Node_CallFunction}
   ) -> (then)
   PlayAnimation(execute=@n7.then, InAnimation=Get HideWidgetAnimation())
   UpdateIconTexture!(
    execute
    self
    InIcon:Image Object Reference=Get CaptureIcon() -> (CaptureIcon:Image Object Reference)
    InSoftTexture:Texture 2D Soft Object Reference=OnCaptureZoneChanged!() -> (OutputDelegate, then, CapturePointIcon:Texture 2D Soft Object Reference) {is=title K2Node_CustomEvent}.CapturePointIcon
   ) -> (then)
  ]}
 UpdateCaptureBarProgress {nodes [
   UpdateCaptureBarProgress!() -> (then, Progress:double=~) {is=title K2Node_FunctionEntry}
   > n1 = Sequence!(execute) -> (then_0, then_1, then_2:Exec)
   n2 = GetOwningPlayer()
   n3 = Get Active Capture Component!(self:CRD Capture Progress On HUDComp Object Reference=Get CaptureProgressOnHUDComponent()) -> (Active Capture Component:Capture Component Object Reference)
   UpdateProgressBarColors!(
    execute=@n1.then_1
    self
    InTeamColors:S UI Creative Core Gamer Team Colors Structure=GetProgressTeamColor!(
     self:PDA UI Creative Core Gamer Object Reference=Get PrimaryDataAsset() -> (PrimaryDataAsset:PDA UI Creative Core Gamer Object Reference)
     CurrentMinigame:Fort Minigame Object Reference=GetMinigame!(ContextActor:Actor Object Reference=@n2) -> (ReturnValue:Fort Minigame Object Reference)
     InLocalPlayerController:Player Controller Object Reference=@n2
     OwningTeam:Byte=Get OwningTeam!(self:Capture Component Object Reference=@n3) -> (OwningTeam:Byte=0)
     CurrentStateTeam:Byte=Get StateOwner!(self:Capture Component Object Reference=@n3) -> (StateOwner:Byte=0)
     CurrentState:ECaptureState Enum=Get MostRecentCaptureState() -> (MostRecentCaptureState=Neutral)
    ) -> (OutColorSet:S UI Creative Core Gamer Team Colors Structure)
   ) -> (then)
   SetScalarParameterValue(
    execute=@n1.then_0
    self=Get CaptureBarMaterial() -> (CaptureBarMaterial:Material Instance Dynamic Object Reference)
    ParameterName=Progress
    Value=Get Progress() -> (Progress:double)
   )
   UpdatePulseState(
    execute=@n1.then_2
    bIsPulseActive=Select!(
     Neutral:bool=false
     Capturing:bool=true
     Contested:bool=false
     Disabled:bool=false
     Decapturing:bool=true
     Neutralizing:bool=true
     Deneutralizing:bool=true
     Captured:bool=false
     Index:ECaptureState Enum=Get MostRecentCaptureState() -> (MostRecentCaptureState=Neutral)
    ) -> (ReturnValue:bool)
   )
   > UpdatePulseDirection!(
    execute
    self
    bIsPulseInverted:bool=Select(
     Neutral=false
     Capturing=false
     Contested=false
     Disabled=false
     Decapturing=true
     Neutralizing=true
     Deneutralizing=false
     Captured=false
     Index=Get MostRecentCaptureState() -> (MostRecentCaptureState=Neutral)
    )
   ) -> (then)
  ]}
 UpdateCaptureBarText {nodes [
   UpdateCaptureBarText() {is=title K2Node_FunctionEntry}
   n1 = Get StatusDescription_Neutral() -> (StatusDescription_Neutral:Text)
   n2 = Get StatusDescription_Capturing() -> (StatusDescription_Capturing:Text)
   n3 = Get StatusDescription_Neutralizing() -> (StatusDescription_Neutralizing:Text)
   > SetText(
    self=Get DialogHeader() -> (DialogHeader:Common Rich Text Block Object Reference)
    InText=Select(
     Neutral:"Text (by ref)"=@n1
     Capturing:"Text (by ref)"=@n2
     Contested:"Text (by ref)"=Get StatusDescription_Contested() -> (StatusDescription_Contested:Text)
     Disabled:"Text (by ref)"=@n1
     Decapturing:"Text (by ref)"=@n3
     Neutralizing:"Text (by ref)"=@n3
     Deneutralizing:"Text (by ref)"=@n2
     Captured:"Text (by ref)"=Get StatusDescription_Captured() -> (StatusDescription_Captured:Text)
     Index=Get MostRecentCaptureState() -> (MostRecentCaptureState=Neutral)
    ) -> (ReturnValue:"Text (by ref)")
   )
  ]}
 UpdateIconTexture {nodes [
   n1 = UpdateIconTexture!() -> (then, InIcon:Image Object Reference, InSoftTexture:Texture 2D Soft Object Reference) {is=title K2Node_FunctionEntry}
   n2 = Get InSoftTexture() -> (InSoftTexture:Texture 2D Soft Object Reference)
   > SetBrushFromSoftTexture!(execute, self:Image Object Reference=@n1, SoftTexture:Texture 2D Soft Object Reference=@n2, bMatchSize:bool=false) -> (then)
   > SetVisibility(
    self=Get InIcon() -> (InIcon:Image Object Reference)
    InVisibility=Select!(
     Option 0:ESlateVisibility Enum=Hidden
     Option 1:ESlateVisibility Enum=Visible
     Index:bool=IsValidSoftObjectReference(SoftObjectReference=@n2)
    ) -> (ReturnValue:ESlateVisibility Enum)
   )
  ]}
 UpdateProgressBarColors {nodes [
   n1 = UpdateProgressBarColors!() -> (
    then
    InTeamColors_Lighter_20_#2:Linear Color Structure
    InTeamColors_Light_21_#3:Linear Color Structure
    InTeamColors_Medium_22_#4:Linear Color Structure
    InTeamColors_Dark_23_#5:Linear Color Structure
    InTeamColors_Darker_24_#6:Linear Color Structure
    InTeamColors_Darker1_27_#7:Linear Color Structure
   ) {is=title K2Node_FunctionEntry}
   n2 = Get CaptureBarMaterial()
   > SetVectorParameterValue(self=@n2, ParameterName=Progress Color 1, Value=@n1.InTeamColors_Medium_22_#4)
   > SetVectorParameterValue(self=@n2, ParameterName=Progress Color 2, Value=@n1)
   > SetVectorParameterValue(self=@n2, ParameterName=Background Color, Value=@n1.InTeamColors_Darker1_27_#7)
  ]}
 UpdatePulseDirection {nodes [
   n1 = UpdatePulseDirection!() -> (then, bIsPulseInverted:bool) {is=title K2Node_FunctionEntry}
   > SetScalarParameterValue(self=Get CaptureBarMaterial(), ParameterName=Invert Pulse Direction, Value="To Float (Boolean)"(InBool=@n1))
  ]}
 UpdatePulseState {nodes [
   n1 = UpdatePulseState!() -> (then, bIsPulseActive:bool=true) {is=title K2Node_FunctionEntry}
   > SetScalarParameterValue(self=Get CaptureBarMaterial(), ParameterName=Active Pulse, Value="To Float (Boolean)"(InBool=@n1))
  ]}
]
variables [
 CaptureBarMaterial {type=MaterialInstanceDynamic}
 CaptureProgressOnHUDComponent {type=CRD_CaptureProgressOnHUDComp_C, default=null}
 DelayBeforeHidingOnCapture {type=number, default=2}
 MostRecentCaptureState {type=ECaptureState, default=Neutral}
 PlayHideAnimTimerHandle {type=TimerHandle}
 PrimaryDataAsset {type=PDA_UI_Creative_CoreGamer_C, default=/Game/Athena/HUD/LTM/Playground/CoreGamer/Data/DA_UI_Creative_CoreGamer}
 StatusDescription_Captured {type=string, default=CAPTURED}
 StatusDescription_Capturing {type=string, default=CAPTURING}
 StatusDescription_Contested {type=string, default=CONTESTED}
 StatusDescription_Neutral {type=string, default=NEUTRAL}
 StatusDescription_Neutralizing {type=string, default=NEUTRALIZING}
 TeamColors {default {
   Friendly {
    backingColor=<rgba 0.004 0.0232 0.2159 1>
    borderColor=<rgba 0.0513 0.7379 1 1>
    contentColorAndOpacity=<rgba 0.5029 1 1 1>
    fillColor=<rgba 0 0.402 1 1>
   }
   Hostile {
    backingColor=<rgba 0.1845 0 0 1>
    borderColor=<rgba 1 0.402 0.0999 1>
    contentColorAndOpacity=<rgba 1 0.7913 0.5029 1>
    fillColor=<rgba 0.8796 0.107 0.0212 1>
   }
   Neutral {
    backingColor=<rgba 0.0243 0.0243 0.0243 1>
    borderColor=<rgba 1 1 1 1>
    contentColorAndOpacity=<rgba 1 1 1 1>
    fillColor=<rgba 0.5573 0.5573 0.5573 1>
   }
  }}
]
widgets {
 info {inheritedWidgetCount=0, parentClass=&CommonActivatableWidget, rootWidgetClass=&Overlay, widgetCount=12}
 widgets [
  RootOverlay {widgetClassPath=&Overlay, bIsVariable=true, widget=RootOverlay}
  WidgetSizeBox {
   widgetClassPath=&SizeBox
   bIsVariable=true
   parent=RootOverlay
   widget=WidgetSizeBox
   properties {bOverride_MinDesiredWidth=true, heightOverride=68, minDesiredWidth=320, widthOverride=320}
   slot {class=OverlaySlot, properties {horizontalAlignment=HAlign_Center, verticalAlignment=VAlign_Bottom, padding=<ltrb 0 0 0 200>}}
  }
  VerticalBox_159 {widgetClassPath=&VerticalBox, parent=WidgetSizeBox, widget=VerticalBox_159, slot {class=SizeBoxSlot}}
  StatusContainer {
   widgetClassPath=&HorizontalBox
   parent=VerticalBox_159
   widget=StatusContainer
   slot {class=VerticalBoxSlot, properties {padding=<ltrb 0 0 0 6>}}
  }
  DialogHeader {
   widgetClassPath=&CommonRichTextBlock
   bIsVariable=true
   parent=StatusContainer
   widget=DialogHeader
   properties {
    bOverrideDefaultStyle=true
    justification=Center
    text=--STATUS--
    wrappingPolicy=AllowPerCharacterWrapping
    defaultTextStyleOverride {
     overflowPolicy=Clip
     transformPolicy=None
     colorAndOpacity {colorUseRule=UseColor_Specified, specifiedColor=<rgba 1 1 1 1>}
     font {
      bForceMonospaced=false
      bMaterialIsStencil=false
      fontMaterial=null
      fontObject=&BurbankSmall
      letterSpacing=0
      monospacedWidth=1
      size=16
      skewAmount=0
      typefaceFontName=Black
      outlineSettings {
       bApplyOutlineToDropShadows=true
       bMiteredCorners=false
       bSeparateFillAlpha=true
       outlineMaterial=&M-FontShadow
       outlineSize=2
       outlineColor=<rgba 0 0 0 1>
      }
     }
     highlightColor {colorUseRule=UseColor_Specified, specifiedColor=<rgba 0 0 0 1>}
     highlightShape {
      bIsDynamicallyLoaded=false
      drawAs=Image
      imageType=NoImage
      mirroring=NoMirror
      resourceName=None
      resourceObject=null
      tiling=NoTile
      imageSize=<32 32>
      margin=<ltrb 0 0 0 0>
      outlineSettings {
       bUseBrushTransparency=false
       roundingType=HalfHeightRadius
       width=0
       color {colorUseRule=UseColor_Specified, specifiedColor=<rgba 0 0 0 0>}
       cornerRadii=<0 0 0 0>
      }
      tintColor {colorUseRule=UseColor_Specified, specifiedColor=<rgba 1 1 1 1>}
      uVRegion {bIsValid=false, max=<0 0>, min=<0 0>}
     }
     selectedBackgroundColor {colorUseRule=UseColor_Foreground, specifiedColor=<rgba 1 0 1 1>}
     shadowColorAndOpacity=<rgba 0 0 0 0>
     shadowOffset=<0 0>
     strikeBrush {
      bIsDynamicallyLoaded=false
      drawAs=Image
      imageType=NoImage
      mirroring=NoMirror
      resourceName=None
      resourceObject=null
      tiling=NoTile
      imageSize=<32 32>
      margin=<ltrb 0 0 0 0>
      outlineSettings {
       bUseBrushTransparency=false
       roundingType=HalfHeightRadius
       width=0
       color {colorUseRule=UseColor_Specified, specifiedColor=<rgba 0 0 0 0>}
       cornerRadii=<0 0 0 0>
      }
      tintColor {colorUseRule=UseColor_Specified, specifiedColor=<rgba 1 1 1 1>}
      uVRegion {bIsValid=false, max=<0 0>, min=<0 0>}
     }
     underlineBrush {
      bIsDynamicallyLoaded=false
      drawAs=Image
      imageType=NoImage
      mirroring=NoMirror
      resourceName=None
      resourceObject=null
      tiling=NoTile
      imageSize=<32 32>
      margin=<ltrb 0 0 0 0>
      outlineSettings {
       bUseBrushTransparency=false
       roundingType=HalfHeightRadius
       width=0
       color {colorUseRule=UseColor_Specified, specifiedColor=<rgba 0 0 0 0>}
       cornerRadii=<0 0 0 0>
      }
      tintColor {colorUseRule=UseColor_Specified, specifiedColor=<rgba 1 1 1 1>}
      uVRegion {bIsValid=false, max=<0 0>, min=<0 0>}
     }
    }
   }
   slot {class=HorizontalBoxSlot, properties {horizontalAlignment=HAlign_Left, verticalAlignment=VAlign_Center, size {sizeRule=Fill, value=1}}}
  }
  Image_165 {
   widgetClassPath=&Image
   parent=StatusContainer
   widget=Image_165
   properties {
    visibility=Collapsed
    brush {
     bIsDynamicallyLoaded=false
     drawAs=Image
     imageType=NoImage
     mirroring=NoMirror
     resourceName=None
     resourceObject=/Game/UI/Sidebar/Textures/T_UI_Icon_Sidebar_OnlinePlayers
     tiling=NoTile
     imageSize=<26 26>
     margin=<ltrb 0 0 0 0>
     outlineSettings {
      bUseBrushTransparency=false
      roundingType=HalfHeightRadius
      width=0
      color {colorUseRule=UseColor_Specified, specifiedColor=<rgba 0 0 0 0>}
      cornerRadii=<0 0 0 0>
     }
     tintColor {colorUseRule=UseColor_Specified, specifiedColor=<rgba 1 1 1 1>}
     uVRegion {bIsValid=false, max=<0 0>, min=<0 0>}
    }
   }
   slot {class=HorizontalBoxSlot, properties {horizontalAlignment=HAlign_Right}}
  }
  Overlay_0 {
   widgetClassPath=&Overlay
   parent=StatusContainer
   widget=Overlay_0
   slot {class=HorizontalBoxSlot, properties {horizontalAlignment=HAlign_Right, verticalAlignment=VAlign_Center, padding=<ltrb 5 0 8 0>}}
  }
  CaptureIcon_FakeOutline {
   widgetClassPath=&Image
   bIsVariable=true
   parent=Overlay_0
   widget=CaptureIcon_FakeOutline
   properties {
    visibility=HitTestInvisible
    brush {
     bIsDynamicallyLoaded=false
     drawAs=Image
     imageType=NoImage
     mirroring=NoMirror
     resourceName=None
     resourceObject=/Game/Athena/HUD/Creative/T_HandIcon
     tiling=NoTile
     imageSize=<27 26>
     margin=<ltrb 0 0 0 0>
     outlineSettings {
      bUseBrushTransparency=false
      roundingType=HalfHeightRadius
      width=0
      color {colorUseRule=UseColor_Specified, specifiedColor=<rgba 0 0 0 0>}
      cornerRadii=<0 0 0 0>
     }
     tintColor {colorUseRule=UseColor_Specified, specifiedColor=<rgba 0 0 0 0.4>}
     uVRegion {bIsValid=false, max=<0 0>, min=<0 0>}
    }
   }
   slot {class=OverlaySlot, properties {horizontalAlignment=HAlign_Center, verticalAlignment=VAlign_Center, padding=<ltrb -0.25 0 0 0>}}
  }
  CaptureIcon {
   widgetClassPath=&Image
   bIsVariable=true
   parent=Overlay_0
   widget=CaptureIcon
   properties {
    visibility=HitTestInvisible
    brush {
     bIsDynamicallyLoaded=false
     drawAs=Image
     imageType=NoImage
     mirroring=NoMirror
     resourceName=None
     resourceObject=&T_HandIcon
     tiling=NoTile
     imageSize=<24 24>
     margin=<ltrb 0 0 0 0>
     outlineSettings {
      bUseBrushTransparency=false
      roundingType=HalfHeightRadius
      width=0
      color {colorUseRule=UseColor_Specified, specifiedColor=<rgba 0 0 0 0>}
      cornerRadii=<0 0 0 0>
     }
     tintColor {colorUseRule=UseColor_Specified, specifiedColor=<rgba 1 1 1 1>}
     uVRegion {bIsValid=false, max=<0 0>, min=<0 0>}
    }
   }
   slot {class=OverlaySlot, properties {horizontalAlignment=HAlign_Center, verticalAlignment=VAlign_Center}}
  }
  Text_Multiplier {
   widgetClassPath=&CommonRichTextBlock
   parent=StatusContainer
   widget=Text_Multiplier
   properties {
    bOverrideDefaultStyle=true
    justification=Center
    text=x
    textTransformPolicy=ToLower
    wrappingPolicy=AllowPerCharacterWrapping
    defaultTextStyleOverride {
     overflowPolicy=Clip
     transformPolicy=None
     colorAndOpacity {colorUseRule=UseColor_Specified, specifiedColor=<rgba 1 1 1 1>}
     font {
      bForceMonospaced=false
      bMaterialIsStencil=false
      fontMaterial=null
      fontObject=&BurbankSmall
      letterSpacing=0
      monospacedWidth=1
      size=16
      skewAmount=0
      typefaceFontName=Black
      outlineSettings {
       bApplyOutlineToDropShadows=true
       bMiteredCorners=false
       bSeparateFillAlpha=true
       outlineMaterial=&M-FontShadow
       outlineSize=2
       outlineColor=<rgba 0 0 0 1>
      }
     }
     highlightColor {colorUseRule=UseColor_Specified, specifiedColor=<rgba 0 0 0 1>}
     highlightShape {
      bIsDynamicallyLoaded=false
      drawAs=Image
      imageType=NoImage
      mirroring=NoMirror
      resourceName=None
      resourceObject=null
      tiling=NoTile
      imageSize=<32 32>
      margin=<ltrb 0 0 0 0>
      outlineSettings {
       bUseBrushTransparency=false
       roundingType=HalfHeightRadius
       width=0
       color {colorUseRule=UseColor_Specified, specifiedColor=<rgba 0 0 0 0>}
       cornerRadii=<0 0 0 0>
      }
      tintColor {colorUseRule=UseColor_Specified, specifiedColor=<rgba 1 1 1 1>}
      uVRegion {bIsValid=false, max=<0 0>, min=<0 0>}
     }
     selectedBackgroundColor {colorUseRule=UseColor_Foreground, specifiedColor=<rgba 1 0 1 1>}
     shadowColorAndOpacity=<rgba 0 0 0 0>
     shadowOffset=<0 0>
     strikeBrush {
      bIsDynamicallyLoaded=false
      drawAs=Image
      imageType=NoImage
      mirroring=NoMirror
      resourceName=None
      resourceObject=null
      tiling=NoTile
      imageSize=<32 32>
      margin=<ltrb 0 0 0 0>
      outlineSettings {
       bUseBrushTransparency=false
       roundingType=HalfHeightRadius
       width=0
       color {colorUseRule=UseColor_Specified, specifiedColor=<rgba 0 0 0 0>}
       cornerRadii=<0 0 0 0>
      }
      tintColor {colorUseRule=UseColor_Specified, specifiedColor=<rgba 1 1 1 1>}
      uVRegion {bIsValid=false, max=<0 0>, min=<0 0>}
     }
     underlineBrush {
      bIsDynamicallyLoaded=false
      drawAs=Image
      imageType=NoImage
      mirroring=NoMirror
      resourceName=None
      resourceObject=null
      tiling=NoTile
      imageSize=<32 32>
      margin=<ltrb 0 0 0 0>
      outlineSettings {
       bUseBrushTransparency=false
       roundingType=HalfHeightRadius
       width=0
       color {colorUseRule=UseColor_Specified, specifiedColor=<rgba 0 0 0 0>}
       cornerRadii=<0 0 0 0>
      }
      tintColor {colorUseRule=UseColor_Specified, specifiedColor=<rgba 1 1 1 1>}
      uVRegion {bIsValid=false, max=<0 0>, min=<0 0>}
     }
    }
   }
   slot {class=HorizontalBoxSlot, properties {horizontalAlignment=HAlign_Right, padding=<ltrb 0 0 4 0>}}
  }
  PlayerCount {
   widgetClassPath=&CommonRichTextBlock
   bIsVariable=true
   parent=StatusContainer
   widget=PlayerCount
   properties {
    bOverrideDefaultStyle=true
    justification=Center
    text=5
    wrappingPolicy=AllowPerCharacterWrapping
    defaultTextStyleOverride {
     overflowPolicy=Clip
     transformPolicy=None
     colorAndOpacity {colorUseRule=UseColor_Specified, specifiedColor=<rgba 1 1 1 1>}
     font {
      bForceMonospaced=false
      bMaterialIsStencil=false
      fontMaterial=null
      fontObject=&BurbankSmall
      letterSpacing=0
      monospacedWidth=1
      size=16
      skewAmount=0
      typefaceFontName=Black
      outlineSettings {
       bApplyOutlineToDropShadows=true
       bMiteredCorners=false
       bSeparateFillAlpha=true
       outlineMaterial=&M-FontShadow
       outlineSize=2
       outlineColor=<rgba 0 0 0 1>
      }
     }
     highlightColor {colorUseRule=UseColor_Specified, specifiedColor=<rgba 0 0 0 1>}
     highlightShape {
      bIsDynamicallyLoaded=false
      drawAs=Image
      imageType=NoImage
      mirroring=NoMirror
      resourceName=None
      resourceObject=null
      tiling=NoTile
      imageSize=<32 32>
      margin=<ltrb 0 0 0 0>
      outlineSettings {
       bUseBrushTransparency=false
       roundingType=HalfHeightRadius
       width=0
       color {colorUseRule=UseColor_Specified, specifiedColor=<rgba 0 0 0 0>}
       cornerRadii=<0 0 0 0>
      }
      tintColor {colorUseRule=UseColor_Specified, specifiedColor=<rgba 1 1 1 1>}
      uVRegion {bIsValid=false, max=<0 0>, min=<0 0>}
     }
     selectedBackgroundColor {colorUseRule=UseColor_Foreground, specifiedColor=<rgba 1 0 1 1>}
     shadowColorAndOpacity=<rgba 0 0 0 0>
     shadowOffset=<0 0>
     strikeBrush {
      bIsDynamicallyLoaded=false
      drawAs=Image
      imageType=NoImage
      mirroring=NoMirror
      resourceName=None
      resourceObject=null
      tiling=NoTile
      imageSize=<32 32>
      margin=<ltrb 0 0 0 0>
      outlineSettings {
       bUseBrushTransparency=false
       roundingType=HalfHeightRadius
       width=0
       color {colorUseRule=UseColor_Specified, specifiedColor=<rgba 0 0 0 0>}
       cornerRadii=<0 0 0 0>
      }
      tintColor {colorUseRule=UseColor_Specified, specifiedColor=<rgba 1 1 1 1>}
      uVRegion {bIsValid=false, max=<0 0>, min=<0 0>}
     }
     underlineBrush {
      bIsDynamicallyLoaded=false
      drawAs=Image
      imageType=NoImage
      mirroring=NoMirror
      resourceName=None
      resourceObject=null
      tiling=NoTile
      imageSize=<32 32>
      margin=<ltrb 0 0 0 0>
      outlineSettings {
       bUseBrushTransparency=false
       roundingType=HalfHeightRadius
       width=0
       color {colorUseRule=UseColor_Specified, specifiedColor=<rgba 0 0 0 0>}
       cornerRadii=<0 0 0 0>
      }
      tintColor {colorUseRule=UseColor_Specified, specifiedColor=<rgba 1 1 1 1>}
      uVRegion {bIsValid=false, max=<0 0>, min=<0 0>}
     }
    }
   }
   slot {class=HorizontalBoxSlot, properties {horizontalAlignment=HAlign_Right}}
  }
  Image_Progress {
   widgetClassPath=&Image
   bIsVariable=true
   parent=VerticalBox_159
   widget=Image_Progress
   properties {
    visibility=HitTestInvisible
    brush {
     bIsDynamicallyLoaded=false
     drawAs=Image
     imageType=NoImage
     mirroring=NoMirror
     resourceName=None
     resourceObject=/Game/Athena/HUD/LTM/Playground/CoreGamer/Materials/MI_UI_Creative_CoreGamer_ProgressBar_Skewed
     tiling=NoTile
     imageSize=<32 10>
     margin=<ltrb 0 0 0 0>
     outlineSettings {
      bUseBrushTransparency=false
      roundingType=HalfHeightRadius
      width=0
      color {colorUseRule=UseColor_Specified, specifiedColor=<rgba 0 0 0 0>}
      cornerRadii=<0 0 0 0>
     }
     tintColor {colorUseRule=UseColor_Specified, specifiedColor=<rgba 1 1 1 1>}
     uVRegion {bIsValid=false, max=<0 0>, min=<0 0>}
    }
   }
   slot {class=VerticalBoxSlot, properties {verticalAlignment=VAlign_Top}}
  }
 ]
}
